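; Blueprint for the EchoFrost Stand B prop, which changes its glass appearance based on whether it has loot.
; The construction script calls the parent, then runs ChangeGlassMaterials.
; That function reads HasLoot and sets both material slots on the stand's glass mesh to their Loot or NoLoot variants.
; BeginPlay, overlap and Tick events are empty, so this graph only sets materials at construction.
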(event EventBeginPlay)

(event Collision|EventActorBeginOverlap
  (params (OtherActor :type Actor%20Object%20Reference)))

(event EventTick
  (params (DeltaSeconds :type Float%20%28single-precision%29)))

(fn ChangeGlassMaterials
  (if (Variables|Default|GetHasLoot)
    (Rendering|Material|SetMaterial (Variables|Default|GetSM_EchoFrost_StandGlass_B) 0 (Variables|Prop|GetM_Slot_0_Loot))
    (Rendering|Material|SetMaterial (Variables|Default|GetSM_EchoFrost_StandGlass_B) 1 (Variables|Prop|GetM_Slot_1_Loot))
    (else
      (Rendering|Material|SetMaterial (Variables|Default|GetSM_EchoFrost_StandGlass_B) 0 (Variables|Prop|GetM_Slot_0_NoLoot))
      (Rendering|Material|SetMaterial (Variables|Default|GetSM_EchoFrost_StandGlass_B) 1 (Variables|Prop|GetM_Slot_1_NoLoot)))))

(fn ConstructionScript
  (|CallParentFunction)
  (CallFunction|ChangeGlassMaterials))
