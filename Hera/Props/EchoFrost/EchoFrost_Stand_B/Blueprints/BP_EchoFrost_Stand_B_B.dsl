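; BP_EchoFrost_Stand_B_B is an EchoFrost trophy stand prop Blueprint that changes the look of its glass to show whether it has loot.
; The construction script runs the parent's, then calls ChangeGlassMaterials, which branches on HasLoot.
; That function sets loot or non-loot ice materials on the glass and interior slots of SM_EchoFrost_StandGlass_B.
; BeginPlay, overlap and Tick events are empty.
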(event EventBeginPlay)

(event Collision|EventActorBeginOverlap
  (params (OtherActor :type Actor%20Object%20Reference)))

(event EventTick
  (params (DeltaSeconds :type Float%20%28single-precision%29)))

(fn ChangeGlassMaterials
  (if (Variables|Default|GetHasLoot)
    (Rendering|Material|SetMaterial (Variables|Default|GetSM_EchoFrost_StandGlass_B) 0 "/Hera/Terrain/Rocks/Materials/EchoFrost/M_EchoFrost_Ice_Baked_Trophies_StandGlass_Loot.M_EchoFrost_Ice_Baked_Trophies_StandGlass_Loot")
    (Rendering|Material|SetMaterial (Variables|Default|GetSM_EchoFrost_StandGlass_B) 1 "/Hera/Terrain/Rocks/Materials/EchoFrost/M_EchoFrost_Ice_Baked_Trophies_Stand_Interior_Loot.M_EchoFrost_Ice_Baked_Trophies_Stand_Interior_Loot")
    (else
      (Rendering|Material|SetMaterial (Variables|Default|GetSM_EchoFrost_StandGlass_B) 0 "/Hera/Terrain/Rocks/Materials/EchoFrost/M_EchoFrost_Ice_Baked_Trophies_StandGlass.M_EchoFrost_Ice_Baked_Trophies_StandGlass")
      (Rendering|Material|SetMaterial (Variables|Default|GetSM_EchoFrost_StandGlass_B) 1 "/Hera/Terrain/Rocks/Materials/EchoFrost/M_EchoFrost_Ice_Baked_Trophies_Stand_Interior.M_EchoFrost_Ice_Baked_Trophies_Stand_Interior"))))

(fn ConstructionScript
  (|CallParentFunction)
  (Default|ChangeGlassMaterials))
